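; Parent Blueprint for Golden Rings touch quest objects, extending the base quest object's initialization and objective-completion events.
; On completion it stores the triggering player, passes it to the parent logic, and plays the ShrinkThing timeline.
; The timeline's popanim track lerps the SM_GoldenRing mesh's world scale, then hides the actor in game when finished.
; Both events also print debug strings, with screen and log output switched off.
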(event EventQuestInitializedAndValid
  (|CallParentFunction)
  (Development|PrintString "Golden Rings Initialized" false false))

(event EventObjectiveSuccessfullyCompleted
  (params (PlayerWhoTriggeredObject :type Actor%20Object%20Reference))
  (bind _output_get (Variables|Default|SetPlayerWhoTriggeredObject PlayerWhoTriggeredObject))
  (|CallParentFunction _output_get)
  (Development|PrintString "Objective completed" false false)
  (bind (_direction _popanim) (Timeline|ShrinkThing PlayFromStart)
    (:Update
      (Transformation|SetWorldScale3D (Variables|Default|GetSM_GoldenRing) (Math|Vector|Lerp(Vector) "0.010000,0.010000,0.010000" "18.000000,18.000000,18.000000" _popanim)))
    (:Finished
      (Rendering|SetActorHiddenInGame true))))

(fn ConstructionScript
  (|CallParentFunction))
